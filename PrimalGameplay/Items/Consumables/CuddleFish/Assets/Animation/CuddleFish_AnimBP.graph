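# Animation Blueprint for the CuddleFish consumable, animating it while held, thrown, jumping, landed and attached to a pawn.
# Each update reads owner tags for jumped, attached and thrown, plus falling and landed flags from the jumper actor.
# While thrown, a looping pose uses an aim offset driven by the projectile's vertical velocity.
# Otherwise, a state machine runs idle (ground or held), jump up, apex loop and pawn landing, beneath a default montage slot.
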AnimBlueprint /PrimalGameplay/Items/Consumables/CuddleFish/Assets/Animation/CuddleFish_AnimBP
target_skeleton=/Game/Gadgets/Assets/Fish/CuddleFish/Meshes/Cuddle_Fish_Fish_Skeleton
graphs [
 AnimGraph {nodes [
   n1 = AimOffset Player 'CuddleFish_Thrown_AO'!(
    X:float=Get ThrowPitch() -> (ThrowPitch:double)
    BasePose:Pose Link Structure=Sequence Player 'CuddleFish_Thrown_Loop_CF'() {sequence=&/CuddleFish_Thrown_Loop_CF}
    Alpha:float=1
   ) -> (Pose:Pose Link Structure) {blendSpace=&/CuddleFish_Thrown_AO, alphaBoolBlend {blendOption=Linear}}
   AnimGraphNode_StateMachine_0 = CuddleFish() {is=title AnimGraphNode_StateMachine}
   n2 = Get IsThrowing() -> (IsThrowing:bool)
   Output Pose(Result=Slot 'DefaultSlot'(Source=Blend Poses by bool(BlendPose_0=@n1, BlendPose_1=@AnimGraphNode_StateMachine_0, bActiveValue=@n2) {blendType=Linear})) {layerGroup=None}
  ]}
 CuddleFish {
  entry=IdleHeld
  path=AnimGraph.AnimGraphNode_StateMachine_0.CuddleFish
  ids=id
  nodes [
   AnimStateNode_0 = IdleHeld() {is=title AnimStateNode}
   AnimStateNode_1 = JumpUp() {is=title AnimStateNode}
   AnimStateNode_2 = JumpLoop() {is=title AnimStateNode}
   AnimStateNode_3 = LandingOnPawn() {is=title AnimStateNode}
   AnimStateTransitionNode_0 = AnimStateTransitionNode() {crossfadeDuration=0.1}
   AnimStateTransitionNode_1 = AnimStateTransitionNode() {bAutomaticRuleBasedOnSequencePlayerInState=true, crossfadeDuration=0.1}
   AnimStateTransitionNode_2 = AnimStateTransitionNode() {crossfadeDuration=0.1}
   AnimStateTransitionNode_3 = AnimStateTransitionNode()
   AnimStateTransitionNode_4 = AnimStateTransitionNode() {crossfadeDuration=0.1}
   AnimStateTransitionNode_6 = AnimStateTransitionNode()
   AnimStateTransitionNode_7 = AnimStateTransitionNode() {crossfadeDuration=0.1}
   AnimStateTransitionNode_8 = AnimStateTransitionNode() {crossfadeDuration=0.1}
  ]
  links [
   {
    dst=JumpLoop
    rule=AnimGraph.AnimGraphNode_StateMachine_0.CuddleFish.AnimStateTransitionNode_4.Transition
    src=IdleHeld
    via=AnimStateTransitionNode_4
   }
   {
    dst=JumpUp
    rule=AnimGraph.AnimGraphNode_StateMachine_0.CuddleFish.AnimStateTransitionNode_0.Transition
    src=IdleHeld
    via=AnimStateTransitionNode_0
   }
   {
    dst=LandingOnPawn
    rule=AnimGraph.AnimGraphNode_StateMachine_0.CuddleFish.AnimStateTransitionNode_8.Transition
    src=IdleHeld
    via=AnimStateTransitionNode_8
   }
   {
    dst=IdleHeld
    rule=AnimGraph.AnimGraphNode_StateMachine_0.CuddleFish.AnimStateTransitionNode_6.Transition
    src=JumpLoop
    via=AnimStateTransitionNode_6
   }
   {
    dst=LandingOnPawn
    rule=AnimGraph.AnimGraphNode_StateMachine_0.CuddleFish.AnimStateTransitionNode_2.Transition
    src=JumpLoop
    via=AnimStateTransitionNode_2
   }
   {
    dst=JumpLoop
    rule=AnimGraph.AnimGraphNode_StateMachine_0.CuddleFish.AnimStateTransitionNode_1.Transition
    src=JumpUp
    via=AnimStateTransitionNode_1
   }
   {
    dst=LandingOnPawn
    rule=AnimGraph.AnimGraphNode_StateMachine_0.CuddleFish.AnimStateTransitionNode_7.Transition
    src=JumpUp
    via=AnimStateTransitionNode_7
   }
   {
    dst=IdleHeld
    rule=AnimGraph.AnimGraphNode_StateMachine_0.CuddleFish.AnimStateTransitionNode_3.Transition
    src=LandingOnPawn
    via=AnimStateTransitionNode_3
   }
  ]
 }
 IdleHeld {
  path=AnimGraph.AnimGraphNode_StateMachine_0.CuddleFish.AnimStateNode_0.IdleHeld
  nodes [Output Animation Pose(Result=Blend Poses by bool(
     BlendPose_0=Sequence Player 'CuddleFish_Idle_Ground_CF'() {sequence=&/CuddleFish_Idle_Ground_CF}
     BlendPose_1=Sequence Player 'CuddleFish_Idle_Held_CF'() {sequence=&/CuddleFish_Idle_Held_CF}
     bActiveValue=Get Has Landed() -> (Has Landed:bool)
    ) {blendType=Linear})]
 }
 JumpUp {
  path=AnimGraph.AnimGraphNode_StateMachine_0.CuddleFish.AnimStateNode_1.JumpUp
  nodes [Output Animation Pose(Result=Sequence Player 'CuddleFish_Jump_Up_CF'() {bLoopAnimation=false, sequence=&/CuddleFish_Jump_Up_CF})]
 }
 JumpLoop {
  path=AnimGraph.AnimGraphNode_StateMachine_0.CuddleFish.AnimStateNode_2.JumpLoop
  nodes [Output Animation Pose(Result=Sequence Player 'CuddleFish_Jump_Apex_Loop_CF'() {sequence=&/CuddleFish_Jump_Apex_Loop_CF})]
 }
 LandingOnPawn {
  path=AnimGraph.AnimGraphNode_StateMachine_0.CuddleFish.AnimStateNode_3.LandingOnPawn
  nodes [Output Animation Pose(Result=Sequence Player 'CuddleFish_Jump_Land_CF'() {bLoopAnimation=false, sequence=&/CuddleFish_Jump_Land_CF})]
 }
 Transition {
  path=AnimGraph.AnimGraphNode_StateMachine_0.CuddleFish.AnimStateTransitionNode_0.Transition
  nodes [Result(bCanEnterTransition=Get Jumped() -> (Jumped:bool))]
 }
 Transition {path=AnimGraph.AnimGraphNode_StateMachine_0.CuddleFish.AnimStateTransitionNode_1.Transition, nodes [Result()]}
 Transition {
  path=AnimGraph.AnimGraphNode_StateMachine_0.CuddleFish.AnimStateTransitionNode_2.Transition
  nodes [Result(bCanEnterTransition=Get HasAttached() -> (HasAttached:bool))]
 }
 Transition {
  path=AnimGraph.AnimGraphNode_StateMachine_0.CuddleFish.AnimStateTransitionNode_3.Transition
  nodes [
   n1 = Get Jumped()
   n2 = Get HasAttached()
   Result(bCanEnterTransition=AND Boolean(A=NOT Boolean(A=@n1), B=NOT Boolean(A=@n2)))
  ]
 }
 Transition {
  path=AnimGraph.AnimGraphNode_StateMachine_0.CuddleFish.AnimStateTransitionNode_4.Transition
  nodes [Result(bCanEnterTransition=Get Is Falling() -> (Is Falling:bool))]
 }
 Transition {
  path=AnimGraph.AnimGraphNode_StateMachine_0.CuddleFish.AnimStateTransitionNode_6.Transition
  nodes [Result(bCanEnterTransition=Get Has Landed())]
 }
 Transition {
  path=AnimGraph.AnimGraphNode_StateMachine_0.CuddleFish.AnimStateTransitionNode_7.Transition
  nodes [Result(bCanEnterTransition=Get HasAttached())]
 }
 Transition {
  path=AnimGraph.AnimGraphNode_StateMachine_0.CuddleFish.AnimStateTransitionNode_8.Transition
  nodes [Result(bCanEnterTransition=Get HasAttached())]
 }
 EventGraph {nodes [
   Event BlueprintUpdateAnimation()
   n1 = GetOwningActor()
   > Is Valid(InputObject=@n1)
   > n2 = Set CuddleFishBP(CuddleFishBP:Actor Object Reference=@n1) -> (Output_Get:Actor Object Reference)
   > n3 = Sequence!(execute) -> (then_0, then_1, then_2:Exec, then_3:Exec)
   n4 = Get Tags!(self:Actor Object Reference=@n2) -> (Tags:Array of Names)
   > Set Jumped(Jumped:bool=Contains Item(TargetArray=@n4, ItemToFind=Jumped)) -> (Output_Get:bool)
   > Set HasAttached(HasAttached:bool=Contains Item(TargetArray=@n4, ItemToFind=AttachedToPawn)) -> (Output_Get:bool)
   > Set IsThrowing(IsThrowing:bool=Contains Item(TargetArray=@n4, ItemToFind=Thrown)) -> (Output_Get:bool)
   n5 = Cast To BGA_Athena_CuddleFish_Jumper!(execute=@n3.then_1, Object=Get CuddleFishBP() -> (CuddleFishBP:Actor Object Reference)) -> (then, CastFailed, AsBGA Athena Cuddle Fish Jumper:BGA Athena Cuddle Fish Jumper Object Reference) {is=title K2Node_DynamicCast}
   > Set Is Falling(Is Falling:bool=Get IsFalling!(self:BGA Athena Cuddle Fish Jumper Object Reference=@n5) -> (IsFalling:bool)) -> (Output_Get:bool)
   > Set Has Landed(Has Landed:bool=Get HasLanded!(self:BGA Athena Cuddle Fish Jumper Object Reference=@n5) -> (HasLanded:bool)) -> (Output_Get:bool)
   n6 = Cast To B_Prj_Athena_Thrown_CuddleFish!(execute=@n3.then_2, Object=Get CuddleFishBP()) -> (then, CastFailed, AsB Prj Athena Thrown Cuddle Fish:B Prj Athena Thrown Cuddle Fish Object Reference) {is=title K2Node_DynamicCast}
   > Set ThrowPitch(ThrowPitch:double=GetVelocity!(self:Actor Object Reference=@n6) -> (ReturnValue_X:double, ReturnValue_Y:double, ReturnValue_Z:double).ReturnValue_Z) -> (Output_Get:double)
  ]}
]
variables [
 CuddleFishBP {type=Actor, default=null}
 Has Landed {type=boolean}
 HasAttached {type=boolean}
 Is Falling {type=boolean}
 IsThrowing {type=boolean}
 Jumped {type=boolean}
 ThrowPitch {type=number}
]
